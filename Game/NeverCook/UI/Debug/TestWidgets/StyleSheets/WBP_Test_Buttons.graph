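# Debug test widget that previews UIKit button styles by size on a dark, centered layout.
# Large, Medium and Small columns each show Loud and Regular buttons, with an extra Quiet button in the Small column that keeps its default size.
# The event graph holds leftover enable-toggle nodes for Button_Loud and SoloButton, which are absent from the widget tree, and no event triggers them.
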WidgetBlueprint /Game/NeverCook/UI/Debug/TestWidgets/StyleSheets/WBP_Test_Buttons
graphs [EventGraph {nodes [
   n1 = Get Button_Loud() -> (Button_Loud:Object Reference)
   SetIsEnabled(self=@n1, bInIsEnabled=NOT Boolean(A=Get bIsEnabled!(self:Widget Object Reference=@n1) -> (bIsEnabled:bool)))
   n2 = Get SoloButton() -> (SoloButton:Solo Button Object Reference)
   > SetIsEnabled(self=@n2, bInIsEnabled=NOT Boolean(A=Get bIsEnabled(self=@n2)))
  ]}]
variables []
widgets {
 info {inheritedWidgetCount=0, parentClass=&UserWidget, rootWidgetClass=&Overlay, widgetCount=16}
 widgets [
  Overlay_32 {widgetClassPath=&Overlay, widget=Overlay_32}
  Background {
   widgetClassPath=&Border
   parent=Overlay_32
   widget=Background
   properties {
    bIsEnabled=false
    background {
     bIsDynamicallyLoaded=false
     drawAs=Image
     imageType=NoImage
     mirroring=NoMirror
     resourceName=None
     resourceObject=null
     tiling=NoTile
     imageSize=<32 32>
     margin=<ltrb 0 0 0 0>
     outlineSettings {
      bUseBrushTransparency=false
      roundingType=HalfHeightRadius
      width=0
      color {colorUseRule=UseColor_Specified, specifiedColor=<rgba 0 0 0 0>}
      cornerRadii=<0 0 0 0>
     }
     tintColor {colorUseRule=UseColor_Specified, specifiedColor=<rgba 0.016 0.016 0.016 1>}
     uVRegion {bIsValid=false, max=<0 0>, min=<0 0>}
    }
   }
   slot {class=OverlaySlot, properties {horizontalAlignment=HAlign_Fill, verticalAlignment=VAlign_Fill}}
  }
  HorizontalBox_0 {
   widgetClassPath=&HorizontalBox
   parent=Overlay_32
   widget=HorizontalBox_0
   slot {class=OverlaySlot, properties {horizontalAlignment=HAlign_Center, verticalAlignment=VAlign_Center}}
  }
  VB_Large {
   widgetClassPath=&VerticalBox
   parent=HorizontalBox_0
   widget=VB_Large
   slot {class=HorizontalBoxSlot, properties {padding=<ltrb 20 0 20 0>}}
  }
  CommonTextBlock_85 {
   widgetClassPath=&CommonTextBlock
   parent=VB_Large
   widget=CommonTextBlock_85
   properties {
    style=&TextStyle-Base_C
    text=Large
    font {
     bForceMonospaced=false
     bMaterialIsStencil=false
     fontMaterial=null
     fontObject=&NotoSans
     letterSpacing=0
     monospacedWidth=1
     size=24
     skewAmount=0
     typefaceFontName=NotoSans-Regular
     outlineSettings {
      bApplyOutlineToDropShadows=false
      bMiteredCorners=false
      bSeparateFillAlpha=false
      outlineMaterial=null
      outlineSize=0
      outlineColor=<rgba 0 0 0 1>
     }
    }
    shadowOffset=<0 0>
   }
   slot {class=VerticalBoxSlot, properties {horizontalAlignment=HAlign_Center, padding=<ltrb 0 0 0 20>}}
  }
  WBP_UIKit_Button_Loud {
   widgetClassPath=/FortUILibrary/UIKit/Buttons/WBP_UIKit_Button_Loud
   bIsVariable=true
   parent=VB_Large
   widget=WBP_UIKit_Button_Loud
   properties {buttonSize=Large}
   slot {class=VerticalBoxSlot, properties {horizontalAlignment=HAlign_Center, padding=<ltrb 10 10 10 10>}}
  }
  WBP_UIKit_Button_Regular {
   widgetClassPath=&WBP_UIKit_Button_Regular
   bIsVariable=true
   parent=VB_Large
   widget=WBP_UIKit_Button_Regular
   properties {buttonSize=Large}
   slot {class=VerticalBoxSlot, properties {horizontalAlignment=HAlign_Center, verticalAlignment=VAlign_Center, padding=<ltrb 10 10 10 10>}}
  }
  VB_Medium {
   widgetClassPath=&VerticalBox
   parent=HorizontalBox_0
   widget=VB_Medium
   slot {class=HorizontalBoxSlot, properties {padding=<ltrb 20 0 20 0>}}
  }
  CommonTextBlock {
   widgetClassPath=&CommonTextBlock
   parent=VB_Medium
   widget=CommonTextBlock
   properties {
    style=&TextStyle-Base_C
    text=Medium
    font {
     bForceMonospaced=false
     bMaterialIsStencil=false
     fontMaterial=null
     fontObject=&NotoSans
     letterSpacing=0
     monospacedWidth=1
     size=24
     skewAmount=0
     typefaceFontName=NotoSans-Regular
     outlineSettings {
      bApplyOutlineToDropShadows=false
      bMiteredCorners=false
      bSeparateFillAlpha=false
      outlineMaterial=null
      outlineSize=0
      outlineColor=<rgba 0 0 0 1>
     }
    }
    shadowOffset=<0 0>
   }
   slot {class=VerticalBoxSlot, properties {horizontalAlignment=HAlign_Center, padding=<ltrb 0 0 0 20>}}
  }
  WBP_UIKit_Button_Loud_1 {
   widgetClassPath=&WBP_UIKit_Button_Loud
   bIsVariable=true
   parent=VB_Medium
   widget=WBP_UIKit_Button_Loud_1
   slot {class=VerticalBoxSlot, properties {horizontalAlignment=HAlign_Center, verticalAlignment=VAlign_Center, padding=<ltrb 10 10 10 10>}}
  }
  WBP_UIKit_Button_Regular_1 {
   widgetClassPath=&WBP_UIKit_Button_Regular
   bIsVariable=true
   parent=VB_Medium
   widget=WBP_UIKit_Button_Regular_1
   slot {class=VerticalBoxSlot, properties {horizontalAlignment=HAlign_Center, verticalAlignment=VAlign_Center, padding=<ltrb 10 10 10 10>}}
  }
  VB_Small {
   widgetClassPath=&VerticalBox
   parent=HorizontalBox_0
   widget=VB_Small
   slot {class=HorizontalBoxSlot, properties {padding=<ltrb 20 0 20 0>}}
  }
  CommonTextBlock_1 {
   widgetClassPath=&CommonTextBlock
   parent=VB_Small
   widget=CommonTextBlock_1
   properties {
    style=&TextStyle-Base_C
    text=Small
    font {
     bForceMonospaced=false
     bMaterialIsStencil=false
     fontMaterial=null
     fontObject=&NotoSans
     letterSpacing=0
     monospacedWidth=1
     size=24
     skewAmount=0
     typefaceFontName=NotoSans-Regular
     outlineSettings {
      bApplyOutlineToDropShadows=false
      bMiteredCorners=false
      bSeparateFillAlpha=false
      outlineMaterial=null
      outlineSize=0
      outlineColor=<rgba 0 0 0 1>
     }
    }
    shadowOffset=<0 0>
   }
   slot {class=VerticalBoxSlot, properties {horizontalAlignment=HAlign_Center, padding=<ltrb 0 0 0 20>}}
  }
  WBP_UIKit_Button_Loud_2 {
   widgetClassPath=&WBP_UIKit_Button_Loud
   bIsVariable=true
   parent=VB_Small
   widget=WBP_UIKit_Button_Loud_2
   properties {buttonSize=Small}
   slot {class=VerticalBoxSlot, properties {horizontalAlignment=HAlign_Center, verticalAlignment=VAlign_Center, padding=<ltrb 10 10 10 10>}}
  }
  WBP_UIKit_Button_Regular_2 {
   widgetClassPath=&WBP_UIKit_Button_Regular
   bIsVariable=true
   parent=VB_Small
   widget=WBP_UIKit_Button_Regular_2
   properties {buttonSize=Small}
   slot {class=VerticalBoxSlot, properties {horizontalAlignment=HAlign_Center, verticalAlignment=VAlign_Center, padding=<ltrb 10 10 10 10>}}
  }
  WBP_UIKit_Button_Quiet {
   widgetClassPath=/FortUILibrary/UIKit/Buttons/WBP_UIKit_Button_Quiet
   bIsVariable=true
   parent=VB_Small
   widget=WBP_UIKit_Button_Quiet
   slot {class=VerticalBoxSlot, properties {horizontalAlignment=HAlign_Center, verticalAlignment=VAlign_Center, padding=<ltrb 10 10 10 10>}}
  }
 ]
}
